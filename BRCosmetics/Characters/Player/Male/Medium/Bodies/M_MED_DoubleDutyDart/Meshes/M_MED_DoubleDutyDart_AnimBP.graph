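# Animation Blueprint for the Double Duty Dart medium male body cosmetic, layering secondary motion and material-driven cloth over the parent pose.
# AnimGraph runs the copied parent pose through a Control Rig fed skydive and targeting values, then offsets backpack and cape attach bones.
# EventGraph drives world-position-offset cloth simulation on base and overlay materials using right and left leg regions, switching wind and turbulence for skydiving.
# The skydive value fed to the rig is always 1.
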AnimBlueprint /BRCosmetics/Characters/Player/Male/Medium/Bodies/M_MED_DoubleDutyDart/Meshes/M_MED_DoubleDutyDart_AnimBP
target_skeleton=&/M_MED_DoubleDutyDart_Skeleton
graphs [
 AnimGraph {nodes [
   n1 = Control Rig!(
    Source=Copy Pose From Mesh() {bUseAttachedParent=true}
    Alpha
    Get Skydive_Alpha:double=Get Get Skydive Alpha() -> (Get Skydive Alpha:double)
    Get Is Target:double=Get Get Is Target() -> (Get Is Target:double)
   ) -> (Pose) {
    alphaBoolBlend {blendOption=Linear}
    controlRigAssetReference {blueprintRigClass=&/M_MED_DoubleDutyDart_ControlRig.M_MED_DoubleDutyDart_ControlRig_C}
   }
   n2 = "Transform (Modify) Bone - Bone: attach_backpack"(ComponentPose=Local To Component(LocalPose=@n1), Alpha=1, Translation="0,0.5,0", Rotation="0,0,0", Scale="1,1,1") {translationMode=BMM_Additive, alphaBoolBlend {blendOption=Linear}, boneToModify {boneName=attach_backpack}, translation {y=0.5}}
   n3 = "Transform (Modify) Bone - Bone: attach_cape"(ComponentPose=@n2, Alpha=1, Translation="0,0,-2", Rotation="0,0,0", Scale="1,1,1") {
    rotationMode=BMM_Additive
    translationMode=BMM_Additive
    alphaBoolBlend {blendOption=Linear}
    boneToModify {boneName=attach_cape}
    translation {z=-2}
   }
   Output Pose(Result=Component To Local(ComponentPose=@n3))
  ]}
 EventGraph {nodes [
   n1 = Event BlueprintUpdateAnimation()
   > n2 = Sequence!(execute) -> (then_0, then_1, then_2:Exec)
   > Set WPO Globals Function(
    Material=Get Mat Instance Dynamic()
    Overlay Material=Get Overlay Mat Instance Dynamic()
    Skeletal Mesh=Get SkelMesh()
    "Local Noise Transform  Location"=Get Local Noise Transform Location() -> (Local Noise Transform Location:Vector)
    Local Noise Offset=Get Local Noise Offset()
   )
   > Set WPO Regional Function!(
    execute
    Material:Material Instance Dynamic Object Reference=Get Mat Instance Dynamic()
    Overlay Material:Material Instance Dynamic Object Reference=Get Overlay Mat Instance Dynamic()
    Skeletal Mesh:Skeletal Mesh Component Object Reference=Get SkelMesh()
    Velocity Transform:Name=foot_r
    Velocity Interp Speed:double=Get Velocity Interp Speed()
    Collision Transform:Name=calf_r
    Collision Transform Axis:EAxis Enum
    Invert Collision Transform Axis:bool=false
    Delta Time X:double=@n1.DeltaTimeX
    Previous Velocity Transform Location:"Vector (by ref)"=Get R1 Velocity Transform Location()
    Previous Velocity Transform Velocity:"Vector (by ref)"=Get R1 Velocity Transform Velocity()
    Region:int=1
   ) -> (then)
   > Set WPO Regional Function(
    Material=Get Mat Instance Dynamic()
    Overlay Material=Get Overlay Mat Instance Dynamic()
    Skeletal Mesh=Get SkelMesh()
    Velocity Transform=foot_l
    Velocity Interp Speed=Get Velocity Interp Speed()
    Collision Transform=calf_l
    Invert Collision Transform Axis=true
    Delta Time X=@n1.DeltaTimeX
    Previous Velocity Transform Location=Get R2 Velocity Transform Location()
    Previous Velocity Transform Velocity=Get R2 Velocity Transform Velocity()
    Region=2
   )
   > n3 = Branch(Condition=Get bIsSkyDiving())
   > Set WPO Pho Wind Intensity(WPO Pho Wind Intensity=0.8)
   > n4 = Set WPO Pho Turbulence(WPO Pho Turbulence:double=1.5) -> (Output_Get:double)
   SetScalarParameterValue(
    execute=[@n6.then, @n4.then]
    self=[Get Overlay Mat Instance Dynamic(), Get Mat Instance Dynamic()]
    ParameterName=ClothSim_WindStrength
    Value=Get WPO Pho Wind Intensity()
   )
   n5 = Get WPO Pho Turbulence() -> (WPO Pho Turbulence:double)
   > SetScalarParameterValue(self=[Get Overlay Mat Instance Dynamic(), Get Mat Instance Dynamic()], ParameterName=ClothSim_R1_TurbulenceStrength, Value=@n5)
   > SetScalarParameterValue(self=[Get Overlay Mat Instance Dynamic(), Get Mat Instance Dynamic()], ParameterName=ClothSim_R2_TurbulenceStrength, Value=@n5)
   Set WPO Pho Wind Intensity(execute=@n3.else, WPO Pho Wind Intensity=0)
   > n6 = Set WPO Pho Turbulence(WPO Pho Turbulence=0.2)
   Set Get Skydive Alpha(execute=@n2.then_1, Get Skydive Alpha:double=1) -> (Output_Get:double)
   Set Get Is Target(execute=@n2.then_2, Get Is Target:double="To Float (Boolean)"(InBool=Get bIsTargeting())) -> (Output_Get:double)
   Event OnCharacterCustomizationLoaded()
   > n7 = Set SkelMesh(SkelMesh=GetPartSkeletalMeshComponent())
   > n8 = Get Dynamic Material Instance(SkeletalMeshComponent=@n7)
   > Set Mat Instance Dynamic(Mat Instance Dynamic=@n8)
   > n9 = GetOrCreateDynamicOverlayMaterial(MeshComponent=@n7, MaterialName=None)
   > Set Overlay Mat Instance Dynamic(Overlay Mat Instance Dynamic=@n9)
   TryGetPawnOwner()
  ]}
]
variables [
 Get Is Target {type=number}
 Get Skydive Alpha {type=number}
 Local Noise Offset {type=Vector}
 Local Noise Transform Location {type=Vector}
 Mat Instance Dynamic {type=MaterialInstanceDynamic}
 Overlay Mat Instance Dynamic {type=MaterialInstanceDynamic}
 R1 Velocity Transform Location {type=Vector}
 R1 Velocity Transform Velocity {type=Vector}
 R2 Velocity Transform Location {type=Vector}
 R2 Velocity Transform Velocity {type=Vector}
 SkelMesh {type=SkeletalMeshComponent}
 Velocity Interp Speed {type=number}
 WPO Pho Turbulence {type=number}
 WPO Pho Wind Intensity {type=number}
]
